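# Post-process material that flattens scene color toward unlit base color, applied after depth of field.
# Desaturated base color, inverted and raised to ShadowTransitionExponent, picks between FlatnessLights and FlatnessDarks as the blend strength, so dark and light albedos flatten independently.
# The blend keeps the scene alpha and applies only on the deferred and default shading paths; forward and mobile pass scene color through unchanged.
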Material /Clyde/TimeOfDay/M_BroadcastPostBaseColorFlatten
graph {
 properties {blendableLocation=BL_SceneColorAfterDOF, materialDomain=MD_PostProcess, pixelDepthOffsetMode=PDOM_Legacy, usedShadingModels=Default Lit}
 nodes [
  n1 = SceneTexture() {sceneTextureId=PPI_PostProcessInput0}
  n2 = ComponentMask(None=SceneTexture() {sceneTextureId=PPI_BaseColor}.Color) {b=true, g=true, r=true}
  n3 = ScalarParameter() {parameterName=FlatnessLights}
  n4 = ScalarParameter() {defaultValue=0.05, parameterName=FlatnessDarks}
  n5 = Clamp(None=Desaturation(None=@n2))
  n6 = MaterialFunctionCall(Float2=Power(Base=OneMinus(None=@n5), Exp=ScalarParameter() {defaultValue=4, parameterName=ShadowTransitionExponent})) {materialFunction=&BreakOutFloat2Components}
  n7 = AppendVector(
   A=LinearInterpolate(A=ComponentMask(None=@n1.Color) {b=true, g=true, r=true}, B=@n2, Alpha=LinearInterpolate(A=@n3, B=@n4, Alpha=@n6.R))
   B=MaterialFunctionCall(Float4=@n1.Color) {materialFunction=&BreakOutFloat4Components}.A
  )
  ShadingPathSwitch(Default=@n7, Deferred=@n7, Forward=@n1.Color, Mobile=@n1.Color)
 ]
}
